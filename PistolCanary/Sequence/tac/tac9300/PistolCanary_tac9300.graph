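# Level Sequence tac9300 in PistolCanary, acting as a container that assembles a shot or scene from subsequences.
# It has seven parallel Subsequence tracks, each with one section spanning the same 0 to 120 frame range at a 30 fps display rate.
# It has no bindings or other tracks of its own, so all content comes from the child sequences, which are not identified here.
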LevelSequence /PistolCanary/Sequence/tac/tac9300/PistolCanary_tac9300
bindings []
display_rate {denominator=1, numerator=30}
tick_resolution {denominator=1, numerator=24000}
tracks [
 {type=MovieSceneSubTrack, display_name=Subsequences, sections [{type=MovieSceneSubSection, end_frame=120, start_frame=0}]}
 {type=MovieSceneSubTrack, display_name=Subsequences, sections [{type=MovieSceneSubSection, end_frame=120, start_frame=0}]}
 {type=MovieSceneSubTrack, display_name=Subsequences, sections [{type=MovieSceneSubSection, end_frame=120, start_frame=0}]}
 {type=MovieSceneSubTrack, display_name=Subsequences, sections [{type=MovieSceneSubSection, end_frame=120, start_frame=0}]}
 {type=MovieSceneSubTrack, display_name=Subsequences, sections [{type=MovieSceneSubSection, end_frame=120, start_frame=0}]}
 {type=MovieSceneSubTrack, display_name=Subsequences, sections [{type=MovieSceneSubSection, end_frame=120, start_frame=0}]}
 {type=MovieSceneSubTrack, display_name=Subsequences, sections [{type=MovieSceneSubSection, end_frame=120, start_frame=0}]}
]
